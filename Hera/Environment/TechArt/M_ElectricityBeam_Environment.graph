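# Additive, unlit, two-sided material that draws environmental electricity beams on ribbon meshes thickened by SplineThicken, with world position offset for height and wobble.
# A generated band, offset by rerouted inputs, forms the arc, alongside panning Perlin noise with per-pivot speed variation and world-aligned plasma noise.
# The band drives depth-faded opacity and a brightened core over a time-shifted three-color gradient with eye adaptation compensation.
# Vertex colors yield stepped masks, while round-bit and micro-noise branches end unconnected.
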Material /Hera/Environment/TechArt/M_ElectricityBeam_Environment
graph {
 properties {
  bAutomaticallySetUsageInEditor=false
  bUsedWithMeshDeformer=true
  bUsedWithSkeletalMesh=true
  bUsedWithSplineMeshes=true
  blendMode=BLEND_Additive
  shadingModel=MSM_Unlit
  twoSided=true
  usedShadingModels=Unlit
 }
 nodes [
  n1 = MaterialFunctionCall!(
   Width=LinearInterpolate(
    A=ScalarParameter() {defaultValue=0.06, parameterName=Electricity_Width_Min}
    B=ScalarParameter() {defaultValue=0.15, parameterName=Electricity_Width_Max}
    Alpha=NamedRerouteUsage()
   )
   Sharpness=ScalarParameter() {defaultValue=0.1, parameterName=Electricity_Sharpness}
   Offset
   Direction Switch
   Compare=Constant() {r=0.5}
   Input Coordinates=Add(A=NamedRerouteUsage(), B=NamedRerouteUsage())
  ) {materialFunction=/Engine/Functions/Engine_MaterialFunctions03/Procedurals/GeneratedBand}
  n2 = Multiply(A=Add(A=@n1.Result, B=NamedRerouteUsage()), B=@n1.Result)
  n3 = Saturate(None=Power(Base=@n2, Exp=ScalarParameter() {defaultValue=2, parameterName=Amp_Exp}) {constExponent=5})
  n4 = ScalarParameter() {defaultValue=1, parameterName=RoundBitsMult}
  n5 = Panner(Coordinate=NamedRerouteUsage(), Speed=AppendVector(A=ScalarParameter() {defaultValue=50, parameterName=CoreMovingSpeed}, B=Constant())) {speedX=0.2}
  n6 = SphereMask(
   A=Fmod(A=ComponentMask(None=@n5) {r=true}, B=ScalarParameter() {defaultValue=100})
   B=Constant() {r=0.5}
   Radius=ScalarParameter() {defaultValue=0.2, parameterName=RoundBitsSize}
  ) {attenuationRadius=0.25, hardnessPercent=0}
  Add(A=@n3, B=Multiply(A=@n2, B=Multiply(A=@n4, B=@n6) {constA=10}))
  Constant() {r=1}
  Constant() {r=1}
  n7 = LinearInterpolate(
   A=ScalarParameter() {parameterName=Electricity_Opacity_Min}
   B=ScalarParameter() {defaultValue=1.5, parameterName=Electricity_Opacity_Max}
   Alpha=NamedRerouteUsage()
  )
  NamedRerouteDeclaration(None=DepthFade(Opacity=Multiply(A=@n7, B=@n2), FadeDistance=ScalarParameter() {defaultValue=200, parameterName=Electricity_DepthFadeDistance})) {name=Electricity_Opacity, nodeColor=<rgba 1 0.0703 0 1>}
  MaterialExpressionTextureCoordinate_5 = TextureCoordinate()
  n8 = Convert(None=@MaterialExpressionTextureCoordinate_5) {
   nodeName=Break Float2
   convertInputs [{
     type=Vector2
     expressionInput {
      expression=&self:MaterialExpressionTextureCoordinate_5
      inputName=None
      mask=0
      maskA=0
      maskB=0
      maskG=0
      maskR=0
      outputIndex=0
     }
    }]
   convertOutputs [{type=Scalar}, {type=Scalar}, {type=Vector2}]
  }
  NamedRerouteDeclaration(None=StaticSwitchParameter(True=@n8, False=@MaterialExpressionTextureCoordinate_5) {defaultValue=true, parameterName=FlipUV}) {name=TexCoord, nodeColor=<rgba 1 0.3516 0 1>}
  n9 = VectorParameter() {parameterName=HeightOffset, defaultValue=<rgba 0 0 0 1>}
  n10 = ScalarParameter() {defaultValue=50, parameterName=RibbonThickness}
  n11 = MaterialFunctionCall(WidthBase=@n10, WidthTip=@n10) {materialFunction=&SplineThicken}
  n12 = Multiply(A=Constant3Vector() {constant=<rgba 0 0 1 0>}, B=ComponentMask(None=NamedRerouteUsage()) {r=true})
  n13 = ScalarParameter() {defaultValue=20, parameterName=Amp_Noise_WPOStrength}
  NamedRerouteDeclaration(None=Add(A=Add(A=@n9.RGB, B=@n11.WorldPositionOffset), B=Multiply(A=@n12, B=@n13))) {name=WPO, nodeColor=<rgba 0 1 0.7188 1>}
  n14 = NamedRerouteUsage()
  n15 = ComponentMask(None=VectorParameter() {parameterName=Amp_NoiseTiling, defaultValue=<rgba 0.001 0.005 0 1>}.RGB) {g=true, r=true}
  n16 = LinearInterpolate(A=@n15, B=Multiply(A=@n15) {constB=1.5}, Alpha=NamedRerouteUsage())
  n17 = Constant()
  n18 = ComponentMask(None=MaterialFunctionCall() {materialFunction=&ObjectPivotPoint}.Object Pivot Location) {g=true}
  n19 = Add(A=Multiply(A=Sine(None=@n18)) {constB=65.48}) {constB=74.5468}
  n20 = Add(
   A=ScalarParameter() {defaultValue=1, parameterName=Amp_NoisePanningSpeed}
   B=Multiply(A=Frac(None=@n19), B=ScalarParameter() {defaultValue=0.2, parameterName=Speed Random Amount})
  )
  n21 = TextureSample(UVs=Panner(Coordinate=Multiply(A=@n14, B=@n16), Speed=AppendVector(A=@n17, B=@n20)) {speedY=3}) {samplerType=SAMPLERTYPE_LinearGrayscale, texture=&T_SimplePerlin_Greyscale}
  n22 = Multiply(
   A=ConstantBiasScale(None=ComponentMask(None=@n21.RGB) {g=true, r=true}) {bias=-0.5, scale=1}
   B=ScalarParameter() {defaultValue=0.5, parameterName=Amp_NoiseStrength}
  ) {constB=0.1}
  n23 = TextureObjectParameter() {
   parameterName=Param
   samplerType=SAMPLERTYPE_LinearGrayscale
   texture=/Game/Accessories/FORT_Backpacks/Backpack_M_MED_StunningMountain/FX/T_BlurredPlasmaCoil
  }
  n24 = Add(A=WorldPosition().XYZ, B=NamedRerouteUsage())
  n25 = Multiply(A=Time(), B=ScalarParameter() {defaultValue=80, parameterName=WorldAlignTexturePanningSpeed})
  n26 = Multiply(
   A=MaterialFunctionCall(
    TextureObject=@n23
    TextureSize=VectorParameter() {parameterName=WorldTextureSize, defaultValue=<rgba 80 80 80 64>}.RGB
    WorldPosition=Add(A=@n24, B=@n25)
   ) {materialFunction=&WorldAlignedTexture}.XY Texture
   B=ScalarParameter() {defaultValue=2, parameterName=Plasma_NoiseStrength}
  ) {constB=0.1}
  NamedRerouteDeclaration(None=Add(
    A=NamedRerouteDeclaration(None=@n22) {name=Amp_Noise, nodeColor=<rgba 0 0.0625 1 1>}
    B=NamedRerouteDeclaration(None=@n26) {name=Plasma_Noise, nodeColor=<rgba 0.3828 1 0 1>}
   )) {name=OverallNoise, nodeColor=<rgba 0.0781 0 1 1>}
  n27 = VertexColor()
  n28 = Constant()
  n29 = Max(
   A=If(A=@n27.G, "A > B"=Constant() {r=0.33}, "A < B"=@n28) {constB=0.5}
   B=If(A=@n27.R, "A > B"=Constant() {r=0.66}, "A < B"=@n28) {constB=0.5}
  )
  n30 = If(A=@n27.B, "A > B"=Constant() {r=1}, "A < B"=@n28) {constB=0.5}
  NamedRerouteDeclaration(None=Saturate(None=Max(A=@n29, B=@n30))) {name=VertexColorMasks, nodeColor=<rgba 0.2422 1 0 1>}
  n31 = ScalarParameter() {defaultValue=3, parameterName=Electricity_EmissiveMult_Min}
  n32 = NamedRerouteUsage()
  n33 = Multiply(A=MaterialFunctionCall(Speed=ScalarParameter() {defaultValue=0.5, parameterName=HueShift_Speed}) {materialFunction=&DebugTimeSine}.Result) {constB=-0.2}
  n34 = Multiply(
   A=MaterialFunctionCall(
    A=VectorParameter() {parameterName=BeamColorA, defaultValue=<rgba 0.974 0 0.4216 1>}.RGB
    B=VectorParameter() {parameterName=BeamColorB, defaultValue=<rgba 0 0.0806 1 1>}.RGB
    C=VectorParameter() {parameterName=BeamColorC, defaultValue=<rgba 0 1.4761 5 1>}.RGB
    Alpha=Add(A=@n32, B=@n33)
   ) {materialFunction=&Lerp_3Color}.Result
   B=If(A=@n3, "A > B"=ScalarParameter() {defaultValue=2, parameterName=BeamCenterBrightening}, "A == B"=@n3, "A < B"=@n3) {constB=0.9}
  )
  NamedRerouteDeclaration(None=EyeAdaptationInverse(LightValueInput=Multiply(A=@n31, B=@n34), AlphaInput=ScalarParameter() {defaultValue=0.7, parameterName=EyeAdaptionInverse}).EyeAdaptationInverse) {name=Emissive, nodeColor=<rgba 1 0.8906 0 1>}
  NamedRerouteUsage()
  NamedRerouteUsage()
  NamedRerouteUsage()
  n35 = VectorParameter() {parameterName=Micro_NoiseTiling, defaultValue=<rgba 0.01 0.5 0 1>}
  n36 = Panner(
   Coordinate=Multiply(A=NamedRerouteUsage(), B=ComponentMask(None=@n35.RGB) {g=true, r=true})
   Speed=AppendVector(A=ScalarParameter() {defaultValue=-0.5, parameterName=Micro_NoisePanningSpeed}, B=Constant())
  ) {speedY=0.25}
  n37 = TextureObject() {texture=/Game/Environments/FrontEnd/Textures/Noise}
  Saturate(None=Add(A=TextureSample(UVs=@n36, Tex=@n37) {samplerType=SAMPLERTYPE_Masks, texture=/Game/Weapons/WeaponSkins/Wraps/Textures/Mash/Mash_Noise}.R) {constB=0.2})
  Saturate(None=@n19)
 ]
}
