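# Translucent, two-sided unlit material for animated environmental energy streams, usable on spline and skeletal meshes.
# World position offset builds camera-facing ribbons, compensates at end-on angles, then adds sine-driven rotation and panning Perlin noise.
# Color uses animated hue shifts on an energy tint, with eye-adaptation-compensated emissive and depth-faded opacity.
# Named reroutes carry procedural bands, panning smoke noise, streaking emissive dots and a vertex-color mask.
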Material /Hera/Environment/TechArt/M_EnergyStreams_Animated_Environment
graph {
 properties {
  bAutomaticallySetUsageInEditor=false
  bUsedWithMeshDeformer=true
  bUsedWithSkeletalMesh=true
  bUsedWithSplineMeshes=true
  blendMode=BLEND_Translucent
  shadingModel=MSM_Unlit
  twoSided=true
  usedShadingModels=Unlit
 }
 nodes [
  n1 = ComponentMask(None=VectorParameter() {parameterName=Amp_NoiseTiling, defaultValue=<rgba 0.01 0.01 0 1>}.RGB) {g=true, r=true}
  LinearInterpolate(A=@n1, B=Multiply(A=@n1) {constB=1.5}, Alpha=NamedRerouteUsage())
  n2 = MaterialFunctionCall(Float2=NamedRerouteUsage()) {materialFunction=&BreakOutFloat2Components}
  n3 = Multiply(A=Time(), B=ScalarParameter() {defaultValue=80, parameterName=ColorPanningSpeed})
  n4 = Add(
   A=Multiply(A=Add(A=@n2.R, B=@n3)) {constB=0.02}
   B=Multiply(A=NamedRerouteUsage(), B=ScalarParameter() {defaultValue=0.3, parameterName=ColorPanningDistortionAmount})
  )
  n5 = LinearInterpolate(
   A=ScalarParameter() {defaultValue=-0.01, parameterName=HueShift_Min}
   B=ScalarParameter() {defaultValue=0.05, parameterName=HueShift_Max}
   Alpha=Power(Base=Sine(None=@n4)) {constExponent=1}
  )
  Multiply(A=@n5)
  n6 = ScalarParameter() {defaultValue=1, parameterName=Bands_Thickness_Max}
  n7 = ScalarParameter() {defaultValue=0.3, parameterName=Bands_Thickness_Min}
  n8 = LocalPosition()
  n9 = Abs(None=Divide(A=ComponentMask(None=@n8.XYZ) {g=true}, B=ScalarParameter() {defaultValue=1, parameterName=Bands_SideMaskDivision}) {constB=1000})
  n10 = ScalarParameter() {defaultValue=1, parameterName=Bands_SideMaskPow}
  n11 = NamedRerouteDeclaration(None=Saturate(None=Power(Base=@n9, Exp=@n10) {constExponent=0.5})) {name=LocalHeightMask, nodeColor=<rgba 0.1953 1 0 1>}
  NamedRerouteDeclaration(None=MaterialFunctionCall!(
    Width=LinearInterpolate(A=@n6, B=@n7, Alpha=@n11) {constA=1, constB=0.001}
    Sharpness=ScalarParameter() {parameterName=Bands_Sharpness}
    Offset
    Direction Switch
    Compare=ScalarParameter() {defaultValue=0.6, parameterName=Bands_Compare}
    Bands=ScalarParameter() {defaultValue=0.5, parameterName=Bands_AmountMult}
    Input Coordinates=NamedRerouteUsage()
   ) {materialFunction=/Engine/Functions/Engine_MaterialFunctions03/Procedurals/GeneratedOffsetBands}.Result) {name=BasicBands, nodeColor=<rgba 0 1 0.7891 1>}
  MaterialExpressionTextureCoordinate_5 = TextureCoordinate()
  n12 = Convert(None=@MaterialExpressionTextureCoordinate_5) {
   nodeName=Break Float2
   convertInputs [{
     type=Vector2
     expressionInput {
      expression=&self:MaterialExpressionTextureCoordinate_5
      inputName=None
      mask=0
      maskA=0
      maskB=0
      maskG=0
      maskR=0
      outputIndex=0
     }
    }]
   convertOutputs [{type=Scalar}, {type=Scalar}, {type=Vector2}]
  }
  NamedRerouteDeclaration(None=StaticSwitchParameter(True=@n12, False=@MaterialExpressionTextureCoordinate_5) {defaultValue=true, parameterName=FlipUV}) {name=TexCoord, nodeColor=<rgba 1 0.3516 0 1>}
  n13 = VectorParameter() {parameterName=MacroNoiseTiling, defaultValue=<rgba 0.15 0.15 0 1>}
  n14 = Multiply(A=NamedRerouteUsage(), B=AppendVector(A=@n13.R, B=@n13.R))
  n15 = VectorParameter() {parameterName=MacroNoisePanningSpeed, defaultValue=<rgba 0.1 0 0 1>}
  NamedRerouteDeclaration(None=TextureSample(UVs=Panner(Coordinate=@n14, Speed=@n15.RGB)) {texture=/Game/Global/Textures/Default/Tileable_Patterns/T_Blowing_Smoke_Pattern2_Tile_DISP}.R) {name=MacroNoise, nodeColor=<rgba 0.2656 1 0 1>}
  n16 = NamedRerouteUsage()
  n17 = ScalarParameter() {defaultValue=-0.01, parameterName=HueShift_Min}
  n18 = ScalarParameter() {defaultValue=0.05, parameterName=HueShift_Max}
  n19 = Sine(None=Multiply(A=@n2.G) {constB=10})
  n20 = VectorParameter() {parameterName=EngergyColor, defaultValue=<rgba 0.0982 0.0282 0.5417 1>}
  n21 = Multiply(
   A=@n16
   B=MaterialFunctionCall(
    Hue Shift Percentage=LinearInterpolate(A=@n17, B=@n18, Alpha=@n19)
    Texture=MaterialFunctionCall(Hue Shift Percentage=@n5, Texture=@n20.RGB) {materialFunction=&HueShift}.Result
   ) {materialFunction=&HueShift}.Result
  )
  n22 = LinearInterpolate(
   A=LinearInterpolate(
    A=ScalarParameter() {defaultValue=8, parameterName=EnergyEmissiveMult}
    B=ScalarParameter() {defaultValue=1, parameterName=EnergyEmissiveMult_Max}
    Alpha=NamedRerouteUsage()
   )
   B=ScalarParameter() {defaultValue=500, parameterName=EmissiveDotEmissiveMult}
   Alpha=NamedRerouteUsage()
  )
  NamedRerouteDeclaration(None=EyeAdaptationInverse(LightValueInput=Multiply(A=@n21, B=@n22), AlphaInput=ScalarParameter() {defaultValue=1, parameterName=EyeSdaptionInverse}).EyeAdaptationInverse) {name=EmissiveColor, nodeColor=<rgba 0 0.0391 1 1>}
  n23 = Multiply(A=NamedRerouteUsage(), B=@n1)
  n24 = AppendVector(A=Constant(), B=ScalarParameter() {defaultValue=0.4, parameterName=Amp_NoisePanningSpeed})
  n25 = ComponentMask(None=TextureSample(UVs=Panner(Coordinate=@n23, Speed=@n24) {speedY=3}) {samplerType=SAMPLERTYPE_LinearGrayscale, texture=&T_SimplePerlin_Greyscale}.RGB) {g=true, r=true}
  NamedRerouteDeclaration(None=Multiply(A=ConstantBiasScale(None=@n25) {bias=-0.5, scale=2}, B=ScalarParameter() {defaultValue=0.15, parameterName=Amp_NoiseStrength}) {constB=0.1}) {name=Amp_Noise, nodeColor=<rgba 0 0.0625 1 1>}
  n26 = Multiply(A=ScalarParameter() {defaultValue=0.1, parameterName=Energy_Opacity}, B=@n16)
  n27 = NamedRerouteUsage()
  NamedRerouteDeclaration(None=Multiply(A=Add(A=@n26, B=@n27), B=DepthFade())) {name=Opacity, nodeColor=<rgba 0.5234 0 1 1>}
  n28 = WorldPosition()
  n29 = Subtract(A=CameraPositionWS(), B=WorldPosition().XYZ)
  n30 = Constant3Vector() {constant=<rgba 0 1 0 0>}
  n31 = Transform(None=Add(B=@n30))
  n32 = Normalize(VectorInput=CrossProduct(A=@n29, B=@n31))
  n33 = Normalize(VectorInput=@n31)
  n34 = Normalize(VectorInput=@n29)
  n35 = Normalize(VectorInput=@n31)
  n36 = Power(Base=Abs(None=DotProduct(A=@n34, B=@n35)), Exp=ScalarParameter() {defaultValue=5, parameterName=HarshPerspectiveGradientPow})
  n37 = ScalarParameter() {defaultValue=20, parameterName=RibbonThickness}
  n38 = LinearInterpolate(A=@n37, B=Multiply(A=@n37) {constB=2}, Alpha=NamedRerouteUsage())
  n39 = TextureCoordinate()
  n40 = StaticSwitch(True=ComponentMask(None=@n39) {g=true}, False=ComponentMask(None=@n39) {r=true}) {defaultValue=true}
  n41 = Add(A=Constant() {r=-0.5}, B=@n39)
  n42 = ComponentMask(None=@n41) {r=true}
  n43 = ComponentMask(None=@n41) {g=true}
  n44 = Multiply(
   A=LinearInterpolate(
    A=@n32
    B=LinearInterpolate(A=@n32, B=@n33) {constAlpha=0.8}
    Alpha=NamedRerouteDeclaration(None=@n36) {name=ExtremedAngleMask, nodeColor=<rgba 0.125 1 0 1>}
   ) {constAlpha=0.9}
   B=Multiply(A=LinearInterpolate(A=@n38, B=@n38, Alpha=@n40), B=StaticSwitch(True=@n42, False=@n43) {defaultValue=true})
  )
  n45 = Add(
   A=Multiply(A=Time(), B=ScalarParameter() {defaultValue=0.3, parameterName=SpinningSpeed})
   B=LinearInterpolate(Alpha=NamedRerouteUsage()) {constA=-1}
  )
  n46 = RotateAboutAxis(
   NormalizedRotationAxis=Transform(None=Constant3Vector() {constant=<rgba 1 0 0 0>}) {transformSourceType=TRANSFORMSOURCE_Local}
   RotationAngle=Multiply(A=Sine(None=@n45), B=ScalarParameter() {defaultValue=0.1, parameterName=SpinningAmount})
   PivotPoint=@n28.XYZ
   Position=Add(A=@n44, B=@n28.XYZ)
  )
  n47 = Multiply(A=Constant3Vector() {constant=<rgba 0 0 1 0>}, B=ScalarParameter() {defaultValue=50, parameterName=Amp_Noise_WPOStrength})
  n48 = ComponentMask(None=NamedRerouteUsage()) {r=true}
  NamedRerouteDeclaration(None=Add(A=Add(A=@n44, B=@n46), B=Multiply(A=@n47, B=@n48))) {name=WPO1, nodeColor=<rgba 0.5938 0 1 1>}
  n49 = VertexColor()
  n50 = Constant()
  n51 = If(A=@n49.G, "A > B"=Constant() {r=0.2}, "A < B"=@n50) {constB=0.5}
  n52 = If(A=@n49.R, "A > B"=Constant() {r=0.4}, "A < B"=@n50) {constB=0.5}
  n53 = Constant() {r=1}
  NamedRerouteDeclaration(None=Max(A=Max(A=@n51, B=@n52), B=If(A=@n49.B, "A > B"=@n53, "A < B"=@n50) {constB=0.5})) {name=VertexColorMask, nodeColor=<rgba 0.2422 1 0 1>}
  n54 = VectorParameter() {parameterName=EmissiveDotsTiling, defaultValue=<rgba 0.06 0.2 0 1>}
  n55 = Multiply(A=NamedRerouteUsage(), B=AppendVector(A=@n54.R, B=@n54.R))
  n56 = TextureSampleParameter2D!(
   UVs=Panner(Coordinate=@n55, Speed=VectorParameter() {parameterName=EmissiveDotsPanningSpeed, defaultValue=<rgba -2 0 0 1>}.RGB)
   "DDX(UVs)"=DDX(Value=@n55)
   "DDY(UVs)"=DDY(Value=@n55)
   Apply View MipBias
  ) {
   desc=Star Texture
   mipValueMode=TMVM_Derivative
   parameterName=EmissiveDotsTexture
   texture=&/T_EmissiveDots_MotionBlur
  }
  NamedRerouteDeclaration(None=@n56.RGB) {name=EmissiveDots, nodeColor=<rgba 0 1 0.3672 1>}
  NamedRerouteUsage()
  NamedRerouteUsage()
  NamedRerouteUsage()
  n57 = Time()
  n58 = ScalarParameter() {defaultValue=0.5, parameterName=AxisSpinningSpeed}
  RotateAboutAxis(
   NormalizedRotationAxis=Constant3Vector() {constant=<rgba 1 0 0 0>}
   RotationAngle=Multiply(A=Add(A=@n57, B=@n58) {constB=0.5}, B=ScalarParameter() {defaultValue=1, parameterName=AxisSpinningAmount})
   PivotPoint=Constant()
   Position=@n30
  )
 ]
}
